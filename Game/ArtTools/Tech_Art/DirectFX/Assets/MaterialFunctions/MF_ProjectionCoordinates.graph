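# DirectFX material function that computes projector-style UVs on world geometry from parameter-driven axes, with location packed in the axes' alpha channels.
# World position relative to the projector is rotated into that basis, with a perspective parameter blending from planar toward direction-based projection.
# Tiling and aspect ratio yield several UV variants.
# It also builds a normal-facing angle mask and square-SDF edge fade, and outputs UVs, fresnel, location, blend vector and ratio.
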MaterialFunction /Game/ArtTools/Tech_Art/DirectFX/Assets/MaterialFunctions/MF_ProjectionCoordinates
graph {nodes [
  FunctionOutput(None=NamedRerouteUsage()) {outputName=UVCoord}
  FunctionOutput(None=NamedRerouteUsage()) {outputName=Projection Fresnel, sortPriority=10}
  FunctionOutput(None=NamedRerouteUsage()) {outputName=Location, sortPriority=20}
  FunctionOutput(None=NamedRerouteUsage()) {outputName=Blend Vector, sortPriority=30}
  FunctionOutput(None=NamedRerouteUsage()) {outputName=UVCoord Unscaled, sortPriority=5}
  FunctionOutput(None=NamedRerouteUsage()) {outputName=X/Y Ratio, sortPriority=100}
  FunctionOutput(None=NamedRerouteUsage()) {outputName=UVCoord Const Ratio, sortPriority=3}
  n1 = WorldPosition() {worldPositionShaderOffset=WPT_ExcludeAllShaderOffsets}
  n2 = WorldPosition()
  n3 = ScalarParameter() {defaultValue=1, group=DFX_Projection, parameterName=DFX_FixedTexture, sliderMax=1}
  n4 = Subtract(
   A=FunctionInput(Preview=LinearInterpolate(A=@n2.XYZ, B=@n1.XYZ, Alpha=@n3)) {bUsePreviewValueAsDefault=true, desc=(Temporary Input - Utopia DFX Screens), inputName=World Position}
   B=NamedRerouteUsage()
  )
  NamedRerouteDeclaration(None=Normalize(VectorInput=@n4)) {name=Projected Forward, nodeColor=<rgba 1 0.7266 0 1>}
  n5 = VectorParameter() {group=DFX_Projection, parameterName=DFX_Matrix_X, defaultValue=<rgba 1 0 0 0>}
  n6 = VectorParameter() {group=DFX_Projection, parameterName=DFX_Matrix_Y, defaultValue=<rgba 0 1 0 0>}
  n7 = VectorParameter() {group=DFX_Projection, parameterName=DFX_Matrix_Z, defaultValue=<rgba 0 0 1 0>}
  NamedRerouteDeclaration(None=MaterialFunctionCall(X=@n5.A, Y=@n6.A, Z=@n7.A) {materialFunction=&MakeFloat3}.Result) {name=Location, nodeColor=<rgba 1 0 0.9141 1>}
  n8 = Constant3Vector() {constant=<rgba 1 1 1 0>}
  NamedRerouteDeclaration(None=ComponentMask(None=MaterialFunctionCall(VectorToTransform=@n8, BasisX=@n5.RGB, BasisY=@n6.RGB, BasisZ=@n7.RGB) {materialFunction=&Transform3x3Matrix}.Transformed Vector) {b=true, g=true}) {name=Ratio, nodeColor=<rgba 0.2422 1 0 1>}
  n9 = LinearInterpolate(Alpha=NamedRerouteUsage()) {constB=0.5}
  n10 = NamedRerouteUsage()
  n11 = LinearInterpolate(Alpha=NamedRerouteUsage()) {constA=-0.5, constB=0}
  n12 = Add(A=@n11) {constB=0.5}
  n13 = Subtract(
   A=Subtract(A=Abs(None=@n10)) {constB=0.5, desc=CompWiseEdgeDist}
   B=LinearInterpolate(A=@n11, B=@n12, Alpha=NamedRerouteUsage()) {constB=0.5}
  )
  n14 = Max(A=@n13) {constB=0}
  n15 = MaterialFunctionCall(Float2=@n13) {materialFunction=&BreakOutFloat2Components}
  n16 = Max(A=@n15.R, B=@n15.R)
  n17 = Add(
   A=Add(A=Length(None=@n14) {is=class MaterialExpressionLength}, B=Min(A=@n16) {constB=0}) {desc=Square SDF}
   B=LinearInterpolate(A=@n12, B=Add(A=@n11), Alpha=NamedRerouteUsage()) {constA=0.5}
  )
  NamedRerouteDeclaration(None=OneMinus(None=SmoothStep(Min=@n9, Value=@n17) {constMax=0.5})) {name=Edge Fade, nodeColor=<rgba 0.0781 0 1 1>}
  n18 = Normalize(VectorInput=@n4)
  n19 = MaterialFunctionCall(
   VectorToTransform=LinearInterpolate(
    A=@n4
    B=Multiply(A=@n18) {constB=100, desc="Arbitrary Mutiplier for\r\nbetter world scale"}
    Alpha=ScalarParameter() {group=DFX_Projection, parameterName=DFX_Perspective, sliderMax=1}
   )
   BasisX=@n5.RGB
   BasisY=@n6.RGB
   BasisZ=@n7.RGB
  ) {materialFunction=&Transform3x3Matrix}
  n20 = Multiply(
   A=NamedRerouteDeclaration(None=ComponentMask(None=@n19.Transformed Vector) {b=true, g=true}) {name=UV_Centered, nodeColor=<rgba 0 1 0.25 1>}
   B=Constant2Vector() {g=-1, r=1}
  ) {constB=-1}
  NamedRerouteDeclaration(None=Add(A=@n20) {constB=0.5}) {name=UV_Unscaled, nodeColor=<rgba 0 1 0.8594 1>}
  n21 = Constant() {r=1}
  n22 = VectorParameter() {group=DFX_Projection, parameterName=DFX_ForwardVector, defaultValue=<rgba 1 0 0 1>}
  n23 = Divide(B=@n22.A)
  NamedRerouteDeclaration(None=If(
    A=@n22.A
    "A > B"=MaterialFunctionCall(X=@n21, Y=@n22.A) {materialFunction=&MakeFloat2}.Result
    "A < B"=MaterialFunctionCall(X=@n23, Y=@n21) {materialFunction=&MakeFloat2}.Result
   ) {constB=1}) {name=X/Y Ratio, nodeColor=<rgba 0.9453 1 0 1>}
  n24 = NamedRerouteUsage()
  n25 = ComponentMask(None=@n20) {r=true}
  n26 = ScalarParameter() {defaultValue=1, group=DFX_Projection, parameterName=DFX_Tiling_X}
  n27 = ComponentMask(None=@n20) {g=true}
  n28 = ScalarParameter() {defaultValue=1, group=DFX_Projection, parameterName=DFX_Tiling_Y}
  n29 = Multiply(
   A=AppendVector(A=Multiply(A=@n25, B=@n26), B=Multiply(A=@n27, B=@n28))
   B=ScalarParameter() {defaultValue=1, group=DFX_Projection, parameterName=DFX_Tiling}
  )
  NamedRerouteDeclaration(None=Add(A=Multiply(A=@n24, B=@n29)) {constB=0.5}) {name=UV_Const Ratio, nodeColor=<rgba 0 1 0.8594 1>}
  n30 = Normalize(VectorInput=@n22.RGB)
  n31 = Normalize(VectorInput=LinearInterpolate(
    A=NamedRerouteDeclaration(None=@n30) {name=Forward Vector, nodeColor=<rgba 0 1 0.1797 1>}
    B=NamedRerouteUsage()
    Alpha=ScalarParameter() {group=DFX_Projection, parameterName=DFX_Perspective, sliderMax=1}
   ))
  n32 = Multiply(A=DotProduct(A=NamedRerouteDeclaration(None=@n31) {name=Blend Vector, nodeColor=<rgba 0 0.6016 1 1>}, B=VertexNormalWS())) {constB=-1}
  NamedRerouteDeclaration(None=SphereMask(
    A=Saturate(None=@n32)
    B=Constant() {r=1}
    Radius=ScalarParameter() {defaultValue=1, group=DFX_Projection, parameterName=DFX_AngleBlend}
    Hardness=ScalarParameter() {defaultValue=0.5, group=DFX_Projection, parameterName=DFX_AngleBlendHardness}
   )) {name=Normalized Dot, nodeColor=<rgba 0 1 0.25 1>}
  NamedRerouteDeclaration(None=Add(A=@n29) {constB=0.5}) {name=UV, nodeColor=<rgba 1 0 0.5859 1>}
  NamedRerouteUsage()
 ]}
